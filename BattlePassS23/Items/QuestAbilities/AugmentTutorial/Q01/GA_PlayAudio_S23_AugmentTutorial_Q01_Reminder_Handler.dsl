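; Passive gameplay ability that plays a voice reminder for the Season 23 Augment tutorial quest (Q01) while the player has augments available to grant.
; On activation it starts a roughly three-minute randomized timer if the controller's augment granting component lists available augments, otherwise it ends.
; When the timer fires, it rechecks augments and quest completion, then applies a random audio GameplayEffect and ends; failed rechecks leave it active without re-arming.
; Ending the ability clears the timer.
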(event Ability|EventFailedToActivatePassiveAbility
  (params (ActorInfo :type Gameplay%20Ability%20Actor%20Info%20Structure))
  (|CallParentFunction ActorInfo))

(event Ability|EventActivateAbility
  (|CallParentFunction)
  (bind _returnvalue (Class|FortControllerComponentAugmentGrantingSystem|GetAugmentGrantingComponentfromController (Fortnite|GetFortPlayerControllerfromActor (Variables|GAVars|Data|GetOwningPlayerPawn))))
  (Utilities|IsValid _returnvalue
    (:Is%20Valid
      (if (> (Utilities|Array|Length (Class|FortControllerComponentAugmentGrantingSystem|GetAvailableAugmentstoGrant _returnvalue)) 0)
        (CallFunction|PlayVo)
        (else
          (Ability|EndAbility))))
    (:Is%20Not%20Valid
      (Ability|EndAbility))))

(event Ability|EventOnEndAbility
  (params (bWasCancelled :type Boolean))
  (Utilities|Time|ClearandInvalidateTimerbyHandle (Variables|Data|GetTimer_Handle)))

(event Custom|Play%20Vo
  (bind _returnvalue (Utilities|Time|SetTimerbyEvent (delegate Custom|Play%20Reminder%20VO) (Math|Random|RandomFloatinRange 1.7e+02 1.9e+02)))
  (Variables|Data|SetTimer_Handle _returnvalue))

(event Custom|Play%20Reminder%20VO
  (bind _returnvalue (Class|FortControllerComponentAugmentGrantingSystem|GetAugmentGrantingComponentfromController (Fortnite|GetFortPlayerControllerfromActor (Variables|GAVars|Data|GetOwningPlayerPawn))))
  (Utilities|IsValid _returnvalue
    (:Is%20Valid
      (if (not (> (Utilities|Array|Length (Class|FortControllerComponentAugmentGrantingSystem|GetAvailableAugmentstoGrant _returnvalue)) 0))
        _
        (elif (Quest|HasCompletedQuest (Variables|GAVars|Data|GetOwningPlayerQuestManager) (Variables|Setup|GetQuestID) :!class /Script/FortniteGame.FortQuestManager)
          _
          (else
            (Ability|ApplyGameplayEffectToOwner (Utilities|Array|RandomArrayItem (Variables|Setup|GetPlay_Audio_GE)))
            (Ability|EndAbility)))))
    (:Is%20Not%20Valid
      _)))
